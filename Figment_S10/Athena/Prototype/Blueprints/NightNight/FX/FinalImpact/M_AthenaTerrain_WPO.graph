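# Translucent unlit terrain effect material for the NightNight final impact, apparently drawing stretched shadow streaks radiating from a CenterPos point.
# Vertex red, weighted by an inverted sphere mask, pushes vertices away from the center by ShadowStretchLength, with an added camera offset.
# Opacity combines inverse vertex red and a radial falloff, flickers through multiplied sine waves of differing periods, and fades with view angle.
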Material /Figment_S10/Athena/Prototype/Blueprints/NightNight/FX/FinalImpact/M_AthenaTerrain_WPO
graph {
 properties {blendMode=BLEND_Translucent, pixelDepthOffsetMode=PDOM_Legacy, shadingModel=MSM_Unlit, twoSided=true, usedShadingModels=Unlit}
 nodes [
  n1 = VertexColor()
  n2 = WorldPosition()
  n3 = VectorParameter() {parameterName=CenterPos, defaultValue=<rgba 66048 -43008 0 0>}
  n4 = ComponentMask(None=@n3.RGB) {b=true, g=true, r=true}
  n5 = SphereMask(A=@n2.XYZ, B=@n4, Radius=LinearInterpolate(Alpha=Constant() {r=1}) {constB=150000}) {hardnessPercent=0}
  n6 = Saturate(None=SquareRoot(None=SquareRoot(None=@n5)))
  n7 = WorldPosition()
  n8 = ScalarParameter() {defaultValue=10, parameterName=ShadowStretchLength}
  n9 = Multiply(
   A=Multiply(A=@n1.R, B=OneMinus(None=@n6))
   B=Multiply(A=Subtract(A=@n7.XYZ, B=@n3.RGB), B=LinearInterpolate(Alpha=@n8) {constAlpha=10, constB=2}) {constB=5000}
  )
  Add(
   A=@n9
   B=MaterialFunctionCall(Offset Amount=ScalarParameter() {parameterName=CameraOffset}, World Position=@n9) {materialFunction=/Engine/Functions/Engine_MaterialFunctions02/WorldPositionOffset/CameraOffset}.Camera Offset
  )
  n10 = ScalarParameter() {defaultValue=1, parameterName=FalloutOpacityRadius}
  n11 = OneMinus(None=@n1.R)
  n12 = Add(A=SphereMask(A=@n2.XYZ, B=@n4, Radius=LinearInterpolate(Alpha=@n10) {constB=500000}) {hardnessPercent=0}, B=@n11)
  n13 = ScalarParameter() {defaultValue=0.3, parameterName=Opacity}
  n14 = Multiply(A=@n13) {constB=0.8}
  n15 = Multiply(A=Time()) {constB=8}
  n16 = Sine(None=@n15) {period=0.155}
  n17 = Sine(None=@n15) {period=0.2}
  n18 = Multiply(A=Sine(None=@n15) {period=0.075}, B=Multiply(A=Multiply(A=@n16, B=@n17), B=Sine(None=@n15) {period=0.43}))
  Multiply(
   A=Multiply(A=Multiply(A=@n12, B=@n11), B=LinearInterpolate(A=@n14, B=@n13, Alpha=@n18))
   B=MaterialFunctionCall("FadeSensitivity (optional)"=ScalarParameter() {defaultValue=4, parameterName=FadeBasedOnViewAngle}) {materialFunction=&FadeBasedOnViewAngle}.Result
  )
  VectorParameter() {parameterName=Color}
 ]
}
