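# Masked, two-sided lit material for scrolling dirt debris in the NightNight final impact helix effect, usable on mesh particles and spline meshes.
# Two panning dirt-clod layers, one tiled by dynamic parameter, screen-blend color and normals, with particle color tinting the switchable emissive.
# Particle alpha erodes their summed alpha, faded near the V=0 edge.
# An optional SplineThicken switch drives world position offset from dynamic width parameters.
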Material /Figment_S10/Athena/Prototype/Blueprints/NightNight/FX/FinalImpact/M_Helix_Debris
graph {
 properties {
  bUsedWithMeshParticles=true
  bUsedWithSplineMeshes=true
  blendMode=BLEND_Masked
  pixelDepthOffsetMode=PDOM_Legacy
  twoSided=true
  usedShadingModels=Default Lit
 }
 nodes [
  n1 = DynamicParameter() {paramNames [X_Width, Y_Width, DebrisPanSpeed, TileXY]}
  n2 = Multiply(A=TextureCoordinate(), B=@n1.TileXY)
  n3 = Add(A=Multiply(A=Time(), B=ScalarParameter() {defaultValue=1, parameterName=InitialPanSpeed}), B=@n1.DebrisPanSpeed)
  n4 = TextureCoordinate()
  MaterialFunctionCall!(
   Blend=TextureSample(UVs=Panner(Coordinate=@n2, Time=@n3) {speedY=-0.3}) {samplerType=SAMPLERTYPE_Normal, texture=/Game/Effects/Fort_Effects/Textures/Dirt/T-DirtClodCluster_N}.RGB
   Base=TextureSample(UVs=Panner(Coordinate=@n4, Time=@n3) {speedY=-0.3}) {samplerType=SAMPLERTYPE_Normal, texture=&T-DirtClodCluster_N}.RGB
  ) {materialFunction=/Engine/Functions/Engine_MaterialFunctions03/Blends/Blend_Screen}
  n5 = ParticleColor()
  n6 = TextureSample(UVs=Panner(Coordinate=@n4, Time=@n3) {speedY=-0.3}) {texture=/Game/Effects/Fort_Effects/Textures/Dirt/T-DirtClodCluster_D}
  n7 = TextureSample(UVs=Panner(Coordinate=@n2, Time=@n3) {speedY=-0.3}) {texture=&T-DirtClodCluster_D}
  n8 = Saturate(None=Add(A=@n6.A, B=@n7.A))
  n9 = MaterialFunctionCall(Float2=TextureCoordinate()) {materialFunction=&BreakOutFloat2Components}
  Multiply(A=Subtract(A=@n5.A, B=OneMinus(None=@n8)), B=Saturate(None=Multiply(A=@n9.G) {constB=4}))
  ScalarParameter() {defaultValue=0.9, parameterName=Roughness}
  StaticSwitchParameter(
   True=Multiply(
    A=@n5.RGB
    B=MaterialFunctionCall(Blend=@n7.RGB, Base=@n6.RGB) {materialFunction=&Blend_Screen}.Result
   )
   False=Constant()
  ) {defaultValue=true, parameterName=UseEmis}
  n10 = DynamicParameter() {paramNames [X_Width, Y_Width, DebrisPanSpeed, TileXY]}
  StaticSwitchParameter(
   True=MaterialFunctionCall(
    WidthBase=LinearInterpolate(Alpha=@n10.X_Width) {constB=50}
    WidthTip=LinearInterpolate(Alpha=@n10.Y_Width) {constB=50}
    Expand U or V UV Channel=StaticBool()
   ) {materialFunction=&SplineThicken}.WorldPositionOffset
   False=Constant()
  ) {parameterName=UseSplineThicken}
 ]
}
