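# Unlit translucent particle material for the NightNight final-impact effect, drawing nebula-and-star smoke tinted by particle color.
# A randomly phased inky smoke flipbook plus soft sphere mask is eroded by a dynamic parameter, then added to stars for opacity, with camera and depth fades.
# Static and panning star samples multiply into bright twinkles over a desaturated nebula.
# Screen-position offsets shift nebula and star UVs; mobile levels use only random offset and skip depth fade.
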Material /Figment_S10/Athena/Prototype/Blueprints/NightNight/FX/FinalImpact/M_NightNight_Nebula
graph {
 properties {
  bUsedWithBeamTrails=true
  bUsedWithParticleSprites=true
  blendMode=BLEND_Translucent
  pixelDepthOffsetMode=PDOM_Legacy
  shadingModel=MSM_Unlit
  usedShadingModels=Unlit
 }
 nodes [
  Constant() {r=0.5}
  n1 = ParticleColor()
  n2 = DynamicParameter() {paramNames [Random,Erode,PanSpeed,MasterAlpha]}
  n3 = TextureCoordinate()
  n4 = Constant() {r=0.5}
  n5 = Constant() {r=2}
  n6 = Add(
   A=Multiply(A=SphereMask(A=@n3, B=@n4, Radius=@n4) {hardnessPercent=0}) {constB=0.15}
   B=TextureSample(UVs=MaterialFunctionCall("Animation  Phase (0-1)"=@n2.Random, Number of Rows=@n5, Number of Columns=@n5) {materialFunction=&FlipBook}.UVs) {texture=/Game/Effects/Fort_Effects/Textures/T_InkySmoke_Flipbook}.G
  )
  n7 = LinearInterpolate(Alpha=Add(A=@n2.Erode, B=@n6)) {constA=-3}
  n8 = Multiply(A=Clamp(None=Multiply(A=@n7, B=@n6))) {constB=0.7}
  n9 = ScreenPosition()
  n10 = ParticlePositionWS()
  n11 = ComponentMask(None=Normalize(VectorInput=Transform(None=@n10) {transformSourceType=TRANSFORMSOURCE_World, transformType=TRANSFORM_View})) {g=true, r=true}
  n12 = Add(A=@n2.Random, B=Multiply(A=Subtract(A=@n9.ViewportUV, B=@n11)) {constB=0.5})
  n13 = FeatureLevelSwitch(Default=@n12, ES2=@n2.Random, ES3_1=@n2.Random, SM4_REMOVED=@n12)
  n14 = TextureSample(UVs=Add(A=@n13, B=TextureCoordinate() {uTiling=4, vTiling=4})) {samplerType=SAMPLERTYPE_Masks, texture=/Game/Characters/Player/Male/Medium/Bodies/M_Med_Soldier_04/Skins/Galaxy/Textures/Celestial_T_Stars}
  n15 = TextureCoordinate() {uTiling=3, vTiling=3}
  n16 = TextureSample(UVs=Panner(Coordinate=Add(A=@n13, B=@n15)) {speedX=0.2, speedY=-0.2}) {samplerType=SAMPLERTYPE_Masks, texture=&Celestial_T_Stars}
  n17 = Multiply(A=Power(Base=@n14.R) {constExponent=1.5}, B=@n16.G)
  n18 = Saturate(None=Multiply(A=@n17) {constB=2})
  n19 = Power(Base=MaterialFunctionCall(Distance=ScalarParameter() {defaultValue=100000, parameterName=ShrinkDistance}) {materialFunction=&Near_Camera_Fade}.Result) {constExponent=3}
  n20 = DepthFade(Opacity=@n19, FadeDistance=ScalarParameter() {defaultValue=10000, parameterName=FadeDist}) {fadeDistanceDefault=12}
  Multiply(A=Multiply(A=@n1.A, B=Add(A=@n8, B=@n18)), B=FeatureLevelSwitch(Default=@n20, ES2=@n19, ES3_1=@n19, SM4_REMOVED=@n20))
  n21 = TextureSample(UVs=Add(A=TextureCoordinate() {uTiling=0.4, vTiling=0.4}, B=@n13)) {texture=/Game/Characters/Player/Male/Medium/Bodies/M_Med_Soldier_04/Skins/Galaxy/Textures/T_FN_Nebula}
  n22 = ScalarParameter() {defaultValue=0.5, parameterName=Desaturation}
  n23 = Add(A=@n21.RGB) {constB=0.3}
  n24 = Multiply(A=@n17) {constB=900}
  Multiply(A=@n1.RGB, B=Add(A=Desaturation(None=@n21.RGB, Fraction=@n22), B=Multiply(A=@n23, B=@n24)))
 ]
}
