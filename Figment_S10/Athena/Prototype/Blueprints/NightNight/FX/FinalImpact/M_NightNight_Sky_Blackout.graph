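# Translucent unlit mesh-particle material for the NightNight final-impact sky blackout effect, previewed on a concave hemisphere.
# A radial UV mask with particle-driven hardness is raised to a power, scaled by particle alpha and softened by a DepthFade with particle-driven distance.
# A color chain multiplies tiled smooth-texture variation by particle color and an inverted, powered radial mask, darkening the center of the UVs.
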Material /Figment_S10/Athena/Prototype/Blueprints/NightNight/FX/FinalImpact/M_NightNight_Sky_Blackout
graph {
 properties {
  bUsedWithMeshParticles=true
  blendMode=BLEND_Translucent
  pixelDepthOffsetMode=PDOM_Legacy
  previewMesh=/Game/Effects/Fort_Effects/Mesh/Smoke/SM_Hemisphere_Concave
  shadingModel=MSM_Unlit
  usedShadingModels=Unlit
 }
 nodes [
  n1 = ParticleColor()
  n2 = Constant() {r=0.5}
  n3 = DynamicParameter() {paramNames [Hardness,DepthFade,Param3,Param4]}
  n4 = SphereMask(A=TextureCoordinate(), B=@n2, Radius=@n2, Hardness=@n3.Hardness) {hardnessPercent=0.6}
  DepthFade(Opacity=Multiply(A=@n1.A, B=Power(Base=@n4) {constExponent=1.5}), FadeDistance=@n3.DepthFade)
  n5 = TextureSample(UVs=TextureCoordinate() {uTiling=3, vTiling=3}) {texture=/Game/Effects/Fort_Effects/Textures/T_Smooth_tile}
  n6 = Constant() {r=0.5}
  n7 = SphereMask(A=TextureCoordinate(), B=@n6, Radius=@n6) {hardnessPercent=55}
  Multiply(A=Multiply(A=LinearInterpolate(Alpha=@n5.G) {constA=0.25}, B=@n1.RGB), B=Power(Base=OneMinus(None=@n7)) {constExponent=3})
 ]
}
