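# Unlit translucent mesh-particle material for pale-blue wind and dust debris in the NightNight final-impact suck-in effect.
# Panned smoke textures at two tilings blend along V and mix into blurred noise, which mobile feature levels skip; a dynamic parameter offsets pan time.
# Opacity combines this pattern with more noise, a UV1 sphere mask, a squared V falloff, particle alpha and depth fade.
# A rotate-about-axis offset around the object position, weighted by inverted V, adds swirl.
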Material /Figment_S10/Athena/Prototype/Blueprints/NightNight/FX/FinalImpact/M_SuckIn_DustWindDebris
graph {
 properties {bUsedWithMeshParticles=true, blendMode=BLEND_Translucent, pixelDepthOffsetMode=PDOM_Legacy, shadingModel=MSM_Unlit, usedShadingModels=Unlit}
 nodes [
  n1 = TextureCoordinate() {uTiling=14}
  n2 = Add(A=Time(), B=DynamicParameter() {paramNames [LandWindPanSpeed, Param2, Param3, Param4]}.LandWindPanSpeed)
  n3 = TextureCoordinate() {uTiling=14}
  n4 = TextureCoordinate() {uTiling=8, vTiling=2}
  n5 = LinearInterpolate(
   A=TextureSample(UVs=Panner(Coordinate=@n3, Time=@n2) {speedX=0.03, speedY=0.07}) {texture=/Figment_S10/Effects/Fort_Effects/Textures/T_Smoke_Puff_Tile_Up}.R
   B=TextureSample(UVs=Panner(Coordinate=@n4, Time=@n2) {speedX=0.03, speedY=0.07}) {texture=&T_Smoke_Puff_Tile_Up}.R
   Alpha=MaterialFunctionCall(Float2=TextureCoordinate()) {materialFunction=&BreakOutFloat2Components}.G
  )
  n6 = LinearInterpolate(
   A=TextureSample(UVs=Panner(Coordinate=@n1, Time=@n2) {speedY=0.2}) {samplerType=SAMPLERTYPE_Grayscale, texture=/Figment_S10/Effects/Fort_Effects/Textures/Tile/LowResBlurredNoise}.R
   B=@n5
  )
  n7 = FeatureLevelSwitch(Default=@n6, ES2=@n5, ES3_1=@n5, SM4_REMOVED=@n6)
  n8 = TextureSample(UVs=Panner(Coordinate=TextureCoordinate() {uTiling=2}, Time=@n2) {speedY=0.2}) {samplerType=SAMPLERTYPE_Grayscale, texture=&LowResBlurredNoise}
  n9 = VectorParameter() {parameterName=SphereMaskCenterOffset, defaultValue=<rgba 0.5 0.5 0 0>}
  n10 = Multiply(
   A=Multiply(A=Multiply(A=@n7, B=@n8.R), B=ScalarParameter() {defaultValue=2, parameterName=Opacity}) {constB=2}
   B=SphereMask(
    A=TextureCoordinate() {coordinateIndex=1}
    B=ComponentMask(None=@n9.RGB) {g=true, r=true}
    Radius=ScalarParameter() {defaultValue=0.5, parameterName=Radius}
    Hardness=ScalarParameter() {parameterName=Hardness}
   ) {attenuationRadius=0.5, hardnessPercent=0}
  )
  n11 = OneMinus(None=MaterialFunctionCall(Float2=TextureCoordinate()) {materialFunction=&BreakOutFloat2Components}.G)
  n12 = Multiply(A=@n11, B=@n11)
  DepthFade(Opacity=Multiply(A=ParticleColor().A, B=Multiply(A=@n10, B=@n12)), FadeDistance=ScalarParameter() {defaultValue=3000, parameterName=DF})
  Multiply(
   A=Power(Base=@n11) {constExponent=2.5}
   B=RotateAboutAxis(
    NormalizedRotationAxis=VectorParameter() {parameterName=SwrilAxis, defaultValue=<rgba 1 0 0 0>}.RGB
    RotationAngle=ScalarParameter() {defaultValue=0.3, parameterName=SwirlStrength}
    PivotPoint=ObjectPositionWS()
    Position=WorldPosition().XYZ
   )
  )
  Multiply(A=Constant3Vector() {constant=<rgba 0.6402 0.8674 1 0>}, B=@n7)
 ]
}
